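# BraveCrown_Wep_AnimBP is the Animation Blueprint for the BraveCrown rifle weapon skeleton.
# A non-looping sequence player holds the BraveCrown_BasePose_Wep pose and feeds a DefaultSlot node into the output, so montages on that slot can play over it.
# The event graph has only unconnected update and pawn-owner nodes, with no variables or logic.
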AnimBlueprint /JethroWeaponGameplay/Gameplay/BraveCrown/Animation/BraveCrown_Wep_AnimBP
target_skeleton=/JethroWeaponGameplay/Gameplay/BraveCrown/Meshes/BraveCrown_Rifle_Skeleton
graphs [
 AnimGraph {nodes [Output Pose(Result=Slot 'DefaultSlot'(Source=Sequence Player 'BraveCrown_BasePose_Wep'() {bLoopAnimation=false, sequence=&/BraveCrown_BasePose_Wep})) {layerGroup=None}]}
 EventGraph {nodes [
   TryGetPawnOwner()
   Event BlueprintUpdateAnimation()
  ]}
]
variables []
